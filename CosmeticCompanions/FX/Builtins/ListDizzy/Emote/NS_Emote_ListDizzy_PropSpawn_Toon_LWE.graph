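# Niagara toon effect, named as the prop-spawn burst for the ListDizzy cosmetic companion emote.
# NE_Spike, the only emitter shown with modules and renderers, combines spawn rates and bursts in four spawn groups, written to VisibilityTag across four sprite renderers.
# Groups alternate bright white and teal, with different sizes, speeds and UV flips.
# Short-lived sprites launch from a small sphere, narrow, lengthen and darken while a curve animates a material parameter.
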NiagaraSystem /CosmeticCompanions/FX/Builtins/ListDizzy/Emote/NS_Emote_ListDizzy_PropSpawn_Toon_LWE
dependencies {
 usedDataInterfaces [
  &NiagaraDataInterfaceSpriteRendererInfo
  &NiagaraDataInterfaceCurve
  &NiagaraDataInterfaceArrayFloat
 ]
 usedDynamicInputs [&FNTimeOfDayColor, &SelectFloatFromArray]
 usedModules [
  &SystemState
  &self:SystemSpawnScript.SystemScriptSource.SystemScriptGraph.NiagaraNodeAssignment_0.SetVariables_#1
 ]
}
emitters {
 AirglowLWE {rendererClasses [], renderers []}
 NE_Spike {
  emitterUpdateScript [
   &EmitterState(Life Cycle Mode=::NewEnumerator1, Loop Behavior=::NewEnumerator1, Loop Delay=0.01, Loop Duration=0.17, UseLoopDelay=true)
   &SpawnRate(SpawnRate=50, Use Spawn Probability=false)
   &SpawnRate(SpawnGroup=1, SpawnRate=20, Use Spawn Probability=false) {module=SpawnRate001}
   &SpawnBurst_Instantaneous(Spawn Count=8, Use Spawn Probability=false, UseLoopCountLimit=false)
   &SpawnBurst_Instantaneous(Spawn Count=7, Spawn Group=1, Spawn Time=0.01, Use Spawn Probability=false, UseLoopCountLimit=false) {module=SpawnBurst_Instantaneous001}
   &SpawnBurst_Instantaneous(Spawn Count=7, Spawn Group=2, Use Spawn Probability=false, UseLoopCountLimit=false) {module=SpawnBurst_Instantaneous002}
   &SpawnBurst_Instantaneous(Spawn Count=8, Spawn Group=3, Spawn Time=0.01, Use Spawn Probability=false, UseLoopCountLimit=false) {module=SpawnBurst_Instantaneous003}
  ]
  particleSpawnScript [
   /SetVariables(Particles.VisibilityTag=&SpawnGroup()) {isSetParameters=true}
   &InitializeParticle(
    Color=&MergeHelper_ColorSwitch(
     0=<rgba 2 2 2 2>
     1=<rgba 0 0.4722 0.5417 1>
     2=<rgba 2 2 2 2>
     3=<rgba 0 0.4722 0.5417 1>
     Index=&SpawnGroup()
     Num=::NewEnumerator2
    )
    Lifetime Max=0.25
    Lifetime Min=0.15
    Lifetime Mode=::NewEnumerator1
    Sprite Size=&MergeHelper_Vector2Switch(
     0=&Multiply_Vector2D_ByFloat(Float=1.5, Vector2D=&RandomRangeVector2D(Minimum=<5 80>, Maximum=<10 110>))
     1=&Multiply_Vector2D_ByFloat(Float=0.8, Vector2D=&RandomRangeVector2D(Minimum=<15 80>, Maximum=<10 100>))
     2=&Multiply_Vector2D_ByFloat(Float=1.5, Vector2D=&RandomRangeVector2D(Minimum=<5 80>, Maximum=<10 110>))
     3=&Multiply_Vector2D_ByFloat(Float=0.8, Vector2D=&RandomRangeVector2D(Minimum=<15 80>, Maximum=<10 100>))
     4=?
     Index=&SpawnGroup()
     Num=::NewEnumerator2
    )
    Sprite Size Mode=::NewEnumerator4
    Sprite UV Mode=::NewEnumerator4
    Sprite UV Scale:Vector2f=&MergeHelper_Vector2Switch(0=<1 1>, 1=<-1 -1>, 2=<1 1>, 3=<-1 -1>, 4=?, Index=&SpawnGroup(), Num=::NewEnumerator2)
    UsePositionOffset=false
    Write Material Random=true
   )
   &CameraOffset(Camera Offset Amount=80)
   &ShapeLocation(MaskSpawnGroup=false, Override Seed=false, Sphere Radius=25, UseOwnerScale=true)
   &AddVelocity(
    Module.Use Falloff Distance=false
    Override Seed=false
    Velocity Mode=::NewEnumerator1
    Velocity Speed=&Multiply_Float(
     A=&RandomRangeFloat(Minimum=1, Maximum=1.5)
     B=&MergeHelper_FloatSwitch(0=200, 1=180, 2=200, 3=180, Index=&SpawnGroup(), Num=::NewEnumerator2)
    )
   )
   &SubUVAnimation(
    Evaluation Type=::NewEnumerator0
    "Sprite Renderer "=DataInterface() {propertyValues {SpriteRenderer=&self:NE_Spike_0.NiagaraSpriteRendererProperties_0}}
    SubUV Animation Mode:ENiagara_SubUVLookupModeV2=::NewEnumerator2
    UseEndFrame=false
    UseStartFrame=false
   )
   &SubUVAnimation(
    Evaluation Type=::NewEnumerator0
    "Sprite Renderer "=DataInterface() {propertyValues {SpriteRenderer=&self:NE_Spike_0.NiagaraSpriteRendererProperties_1}}
    SubUV Animation Mode:ENiagara_SubUVLookupModeV2=::NewEnumerator2
    UseEndFrame=false
    UseStartFrame=false
   ) {module=SubUVAnimation001}
  ]
  particleUpdateScript [
   &ParticleState()
   &ScaleSpriteSize(
    Scale Factor=&MakeVector2D(
     X=&FloatFromCurve(FloatCurve=DataInterface() {propertyValues {Curve {keys [<at 0 1>,<at 0.4756 0.95>,<at 1 0.36>]}}})
     Y=&FloatFromCurve(FloatCurve=DataInterface() {propertyValues {Curve {keys [<at 0 0.25>,<at 1 0.7>]}}})
    )
    Scale Sprite Size Mode=::NewEnumerator2
   )
   &ScaleColor(
    Scale Alpha=&FloatFromCurve(FloatCurve=DataInterface() {propertyValues {Curve {keys [<at 0 1>,<at 0.5056 1>]}}})
    Scale RGB=&VectorFromFloat(Value=&FloatFromCurve(FloatCurve=DataInterface() {propertyValues {Curve {keys [<at 0 1>,<at 0.4043 0.85>,<at 1.0023 0.15>]}}}))
    ScaleA=true
    ScaleRGB=true
   )
   &Drag(Drag=0.2, Use Linear Drag=true, Use Rotational Drag=false)
   &SolveForcesAndVelocity(Clamp Velocity=false, Limit Acceleration=false)
   &DynamicMaterialParameters(
    Index 0 Param 1=1
    Index 0 Param 2=1
    Index 0 Param 4=&FloatFromCurve(FloatCurve=DataInterface() {propertyValues {Curve {defaultValue=1, keys [{tangentMode=RCTM_User, time=0, value=0}, <at 0.2911 0.075>, {tangentMode=RCTM_User, time=1, value=1.25}]}}})
    Module.Index0Float0Write=true
    Module.Index0Float1Write=true
    Module.Index0Float2Write=true
    Module.Index0Float3Write=true
   )
  ]
  rendererClasses [&NiagaraSpriteRendererProperties]
  renderers [
   {rendererClass=&NiagaraSpriteRendererProperties}
   {rendererClass=&NiagaraSpriteRendererProperties, rendererIndex=1}
   {rendererClass=&NiagaraSpriteRendererProperties, rendererIndex=2}
   {rendererClass=&NiagaraSpriteRendererProperties, rendererIndex=3}
  ]
 }
 SparklesLWE {rendererClasses [], renderers []}
 ToonImpactSmoke {rendererClasses [], renderers []}
 ToonSmoke {rendererClasses [], renderers []}
 ToonSmoke_LowMedium {rendererClasses [], renderers []}
}
summary {
 systemName=NS_Emote_ListDizzy_PropSpawn_Toon_LWE
 emitters [
  {emitterName=AirglowLWE, rendererClasses []}
  {emitterName=SparklesLWE, rendererClasses []}
  {emitterName=ToonImpactSmoke, rendererClasses []}
  {emitterName=ToonSmoke, rendererClasses []}
  {emitterName=ToonSmoke_LowMedium, rendererClasses []}
  {emitterName=NE_Spike, rendererClasses [&NiagaraSpriteRendererProperties]}
 ]
 userVariables [User.NormalizedNotifyProgress {type=NiagaraFloat, defaultValue=0, description=""}]
}
